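; Test launch pad trap Blueprint that handles the pad's trigger setup and audiovisual feedback, with no launch physics in its graphs.
; Placement registers the collision component as trigger and plays dust and sound; finished building adds a sound indicator filtered to hostile and neutral affiliations.
; Gameplay cues drive fire and reload sounds, while a launch event activates the pad's particle component.
; Durability loss triggers delayed destruction effects, and server death hides the actor.
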(event EventOnPlaced
  ;[1 1 1] Initial Setup
  (Trap|AddTriggerComponent (Variables|Default|GetS_Athena_Launchpad_Collision))
  (Audio|PlaySoundatLocation (Variables|Audio|GetTrap_Placed_Sound) (Transformation|GetWorldLocation (Variables|Default|GetPlacementSoundLocation)))
  (Utilities|TrapPlacementDustFX :Attach%20to%20Component (Variables|Debug|GetStaticMeshComponent) :Socket "PlacedFX" :Rotation "0, 0, 0" :Burst%20Rate 8 :Size%20%28Volume%29 "100.000000,100.000000,15.000000" :Velocity%20Scale "1,1,1" :Size%20Scale%20%28Particles%29 1))

(event EventOnFinishedBuilding
  ;[1 1 1] Finished building
  (Audio|PlaySoundatLocation (Variables|Audio|GetTrap_Active_Sound) (Transformation|GetActorLocation))
  (Fortnite|ShowSoundIndicatorLocalforTeam :Location (Transformation|GetActorLocation) :MaxAudibleDistance (Variables|/Script/Engine.SoundBase|GetMaxDistance (Variables|Audio|GetTrap_Placed_Sound)) :IndicatorType "Gunshot" :Team (Team|GetActorTeam self) :Affiliations (Utilities|Array|MakeArray "Hostile" "Neutral") :Tint "(R=0.000000,G=0.000000,B=0.000000,A=0.000000)" :OverrideIcon "/Game/UI/Foundation/Textures/Icons/Navigation/t_icon_nav_mobile_audioviz_building_32.t_icon_nav_mobile_audioviz_building_32" :AttenuationMaxRangeMultiplier 1.0 :AttenuationIntensityMultiplier 1.0))

(event GameplayCue|GameplayCue.Abilities.Activation.Traps.ActivateTrap
  (params (EventType :type EGameplayCueEvent%20Enum :default "OnActive") (Parameters :type Gameplay%20Cue%20Parameters%20Structure%20%28by%20ref%29))
  ;[1 1 1] On Fire
  (switch Utilities|FlowControl|Switch|SwitchonEGameplayCueEvent EventType
    (:Executed
      (Audio|PlaySoundatLocation (Variables|Audio|GetTrap_Fire_Sound) (Transformation|GetActorLocation) "0, 0, 0" 1.5 0.75))))

(event GameplayCue|GameplayCue.Abilities.Activation.Traps.ReloadBegin
  (params (EventType :type EGameplayCueEvent%20Enum :default "OnActive") (Parameters :type Gameplay%20Cue%20Parameters%20Structure%20%28by%20ref%29))
  ;[1 1 1] Reload Begin
  (switch Utilities|FlowControl|Switch|SwitchonEGameplayCueEvent EventType
    (:Executed
      (Audio|PlaySoundatLocation (Variables|Audio|GetTrap_Reload_Sound) (Transformation|GetActorLocation)))))

(event EventOnReloadEnd
  ;[1 1 1] Reload End
  (Audio|PlaySoundatLocation (Variables|Audio|GetTrap_Active_Sound) (Transformation|GetActorLocation)))

(event EventOnOutOfDurability
  ;[1 1 1] Durability Change
  (Utilities|FlowControl|Delay 1)
  (Audio|PlaySoundatLocation (Variables|Audio|GetTrap_Explode_Sound) (Transformation|GetActorLocation))
  (Effects|SpawnEmitteratLocation "/Game/Weapons/FORT_TRAPS/Spike_Trap/P_Spike_Trap_Destruction.P_Spike_Trap_Destruction" (Transformation|GetWorldLocation (Variables|Transformation|GetRootComponent))))

(event EventTriggerLaunchEffects
  (params (Pawn :type Fort%20Pawn%20Object%20Reference))
  ;[1 1 1] trigger launch FX
  (Components|Activation|SetActive (Variables|Default|GetP_Athena_LaunchPad_temp) true true))

(event EventOnDeathServer
  (params (Damage :type Float%20%28single-precision%29) (DamageTags :type Gameplay%20Tag%20Container%20Structure%20%28by%20ref%29) (Momentum :type Vector :default "0, 0, 0") (HitInfo :type Hit%20Result%20Structure%20%28by%20ref%29) (InstigatedBy :type Controller%20Object%20Reference) (DamageCauser :type Actor%20Object%20Reference) (EffectContext :type Gameplay%20Effect%20Context%20Handle%20Structure))
  (Rendering|SetActorHiddenInGame true))

(fn ConstructionScript)
